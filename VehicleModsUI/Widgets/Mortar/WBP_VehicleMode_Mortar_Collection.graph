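# Widget Blueprint for the mortar vehicle mod's reticle pip collection, derived from the shared recharge stack pip collection base.
# A fixed-height SizeBox holds a vertically centered DynamicEntryBox that spawns mortar pip widgets with negative horizontal spacing so they overlap.
# On Construct it hands that entry box to the parent through SetDynamicEntryBox, leaving any stack and recharge logic to the base class.
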WidgetBlueprint /VehicleModsUI/Widgets/Mortar/WBP_VehicleMode_Mortar_Collection
graphs [EventGraph {nodes [
   Event Construct()
   > SetDynamicEntryBox!(
    execute
    self:Fort Weapon Reticle Recharge Stack Base Pip Collection Base Object Reference
    DynamicEntryBox:Dynamic Entry Box Object Reference=Get DynamicEntryBox_Stacks() -> (DynamicEntryBox_Stacks:Dynamic Entry Box Object Reference)
   ) -> (then)
  ]}]
variables []
widgets {
 info {
  inheritedWidgetCount=0
  parentClass=/Game/Athena/HUD/reticles/ReticleExtensionWidgets/ChargeStack/WBP_Reticle_RechargeStackBase_PipCollection.WBP_Reticle_RechargeStackBase_PipCollection_C
  rootWidgetClass=&SizeBox
  widgetCount=2
 }
 widgets [
  SizeBox_0 {widgetClassPath=&SizeBox, widget=SizeBox_0, properties {bOverride_HeightOverride=true, heightOverride=48, widthOverride=190}}
  DynamicEntryBox_Stacks {
   widgetClassPath=&DynamicEntryBox
   bIsVariable=true
   parent=SizeBox_0
   widget=DynamicEntryBox_Stacks
   properties {
    entryWidgetClass=&/WBP_VehicleMode_Mortar_Pip.WBP_VehicleMode_Mortar_Pip_C
    numDesignerPreviewEntries=6
    entrySpacing=<-13.0597 0>
   }
   slot {class=SizeBoxSlot, properties {verticalAlignment=VAlign_Center}}
  }
 ]
}
