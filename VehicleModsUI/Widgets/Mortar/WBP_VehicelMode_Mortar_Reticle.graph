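# Mortar vehicle-mod reticle widget, extending the recharge-stack reticle base with a material-driven crosshair, ammo pip collection and disabled warning.
# Binds the weapon's trigger-press and targeting events, with animations for intro, hip/ADS fire and forward/reverse ADS transitions that mostly key reticle material parameters.
# While the magazine is below full, a cooldown timer feeds the mortar weapon's ammo-grant timer into the pips' recharge display.
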WidgetBlueprint /VehicleModsUI/Widgets/Mortar/WBP_VehicelMode_Mortar_Reticle
animations [
 Anim_Intro {
  binding_count=3
  length_seconds=0.5
  object=&self:Anim_Intro
  sequence {bindings [
    Overlay_0 {
     id=#1
     possessed_class=Overlay
     tracks [
      {
       type=MovieSceneFloatTrack
       display_name=Render Opacity
       sections [{type=MovieSceneFloatSection, end_frame=10, start_frame=0, channels [None {type=MovieSceneScriptingFloatChannel, keys [[0,0],[10,1]]}]}]
      }
      {
       type=MovieScene2DTransformTrack
       sections [{
         type=MovieScene2DTransformSection
         end_frame=10
         start_frame=0
         channels [
          Scale.X {type=MovieSceneScriptingFloatChannel, keys [[0,1.1],[10,1]]}
          Scale.Y {type=MovieSceneScriptingFloatChannel, keys [[0,1.1],[10,1]]}
         ]
        }]
      }
     ]
    }
    WBP_VehicleMode_Mortar_Collection {
     id=#2
     possessed_class=WBP_VehicleMode_Mortar_Collection_C
     tracks [{
       type=MovieScene2DTransformTrack
       sections [{
         type=MovieScene2DTransformSection
         end_frame=10
         start_frame=0
         channels [
          Translation.X {type=MovieSceneScriptingFloatChannel, keys [[4,40],[10,38]]}
          Translation.Y {type=MovieSceneScriptingFloatChannel, const [-56,2]}
          Scale.X {type=MovieSceneScriptingFloatChannel, keys [[3.6363,0.4],[10,0.38]]}
          Scale.Y {type=MovieSceneScriptingFloatChannel, keys [[3.6363,0.005],[10,1]]}
         ]
        }]
      }]
    }
    Reticle {
     id=#3
     possessed_class=Image
     tracks [{
       type=MovieSceneWidgetMaterialTrack
       sections [{
         type=MovieSceneParameterSection
         end_frame=10
         start_frame=0
         channels [
          GapSize {type=MovieSceneScriptingFloatChannel, keys [[0,-0.5],[6,-0.8],[10,-0.9]]}
          RectangleSize {type=MovieSceneScriptingFloatChannel, keys [[0,0.22],[6,0.3]]}
          OffsetX {type=MovieSceneScriptingFloatChannel, keys [[3,0.51],[10,-0.06]]}
          BottomScale {type=MovieSceneScriptingFloatChannel, const [0.5,1]}
          BottomPartPosition {type=MovieSceneScriptingFloatChannel, keys [[3,-0.73],[10,-0.46]]}
          OpacityDropLines {type=MovieSceneScriptingFloatChannel, const [0.3,1]}
          DropLinesThickness {type=MovieSceneScriptingFloatChannel, keys [[5,-0.04],[10,0.03]]}
         ]
        }]
      }]
    }
   ]}
 }
 Anim_Fire {
  binding_count=2
  length_seconds=0.5
  object=&self:Anim_Fire
  sequence {bindings [
    WBP_VehicleMode_Mortar_Collection {
     id=#4
     possessed_class=WBP_VehicleMode_Mortar_Collection_C
     tracks [{
       type=MovieScene2DTransformTrack
       sections [{
         type=MovieScene2DTransformSection
         end_frame=10
         start_frame=0
         channels [
          Translation.X {type=MovieSceneScriptingFloatChannel, const [38,1]}
          Translation.Y {type=MovieSceneScriptingFloatChannel, const [-56,1]}
          Scale.X {type=MovieSceneScriptingFloatChannel, const [0.38,1]}
          Scale.Y {type=MovieSceneScriptingFloatChannel, const [1,1]}
         ]
        }]
      }]
    }
    Reticle {
     id=#5
     possessed_class=Image
     tracks [{
       type=MovieSceneWidgetMaterialTrack
       sections [{
         type=MovieSceneParameterSection
         end_frame=10
         start_frame=0
         channels [
          GapSize {type=MovieSceneScriptingFloatChannel, keys [[0,-0.9],[4,-0.75],[10,-0.9]]}
          RectangleSize {type=MovieSceneScriptingFloatChannel, keys [[0,0.3],[4,0.35],[10,0.3]]}
          OffsetX {type=MovieSceneScriptingFloatChannel, keys [[0,-0.06],[4,0.06],[10,-0.06]]}
          BottomScale {type=MovieSceneScriptingFloatChannel, const [0.5,1]}
          BottomPartPosition {type=MovieSceneScriptingFloatChannel, const [-0.46,1]}
          OpacityDropLines {type=MovieSceneScriptingFloatChannel, const [0.3,2]}
          DropLinesThickness {type=MovieSceneScriptingFloatChannel, const [0.03,1]}
          OpacitySideLines {type=MovieSceneScriptingFloatChannel, keys [[0,1],[4,0.3],[10,1]]}
          Opacity {type=MovieSceneScriptingFloatChannel, keys [[0,1],[4,0.3],[8,0.3],[10,1]]}
         ]
        }]
      }]
    }
   ]}
 }
 Anim_ADS {
  binding_count=3
  length_seconds=0.5
  object=&self:Anim_ADS
  sequence {bindings [
    Reticle {
     id=#6
     possessed_class=Image
     tracks [{
       type=MovieSceneWidgetMaterialTrack
       sections [{
         type=MovieSceneParameterSection
         end_frame=10
         start_frame=0
         channels [
          GapSize {type=MovieSceneScriptingFloatChannel, keys [[0,-0.9],[10,-0.95]]}
          RectangleSize {type=MovieSceneScriptingFloatChannel, keys [[0,0.3],[10,0.26]]}
          BottomScale {type=MovieSceneScriptingFloatChannel, keys [[0,0.5],[10,0.36]]}
          BottomPartPosition {type=MovieSceneScriptingFloatChannel, keys [[0,-0.46],[10,-0.36]]}
          OffsetX {type=MovieSceneScriptingFloatChannel, keys [[0,-0.06],[10,-0.24]]}
          OpacityDropLines {type=MovieSceneScriptingFloatChannel, keys [[0,0.3],[10,1]]}
          DropLinesThickness {type=MovieSceneScriptingFloatChannel, keys [[0,0.03],[10,0.05]]}
         ]
        }]
      }]
    }
    WBP_VehicleMode_Mortar_Collection {
     id=#7
     possessed_class=WBP_VehicleMode_Mortar_Collection_C
     tracks [{
       type=MovieScene2DTransformTrack
       sections [{
         type=MovieScene2DTransformSection
         end_frame=10
         start_frame=0
         channels [
          Translation.X {type=MovieSceneScriptingFloatChannel, keys [[0,38],[10,20]]}
          Translation.Y {type=MovieSceneScriptingFloatChannel, const [-56,1]}
          Scale.X {type=MovieSceneScriptingFloatChannel, const [0.38,2]}
          Scale.Y {type=MovieSceneScriptingFloatChannel, const [1,2]}
         ]
        }]
      }]
    }
   ]}
 }
 Anim_Fire_ADS {
  binding_count=2
  length_seconds=0.5
  object=&self:Anim_Fire_ADS
  sequence {bindings [
    WBP_VehicleMode_Mortar_Collection {
     id=#8
     possessed_class=WBP_VehicleMode_Mortar_Collection_C
     tracks [{
       type=MovieScene2DTransformTrack
       sections [{
         type=MovieScene2DTransformSection
         end_frame=10
         start_frame=0
         channels [
          Translation.X {type=MovieSceneScriptingFloatChannel, const [20,1]}
          Translation.Y {type=MovieSceneScriptingFloatChannel, const [-56,1]}
          Scale.X {type=MovieSceneScriptingFloatChannel, const [0.38,1]}
          Scale.Y {type=MovieSceneScriptingFloatChannel, const [1,1]}
         ]
        }]
      }]
    }
    Reticle {
     id=#9
     possessed_class=Image
     tracks [{
       type=MovieSceneWidgetMaterialTrack
       sections [{
         type=MovieSceneParameterSection
         end_frame=10
         start_frame=0
         channels [
          GapSize {type=MovieSceneScriptingFloatChannel, keys [[0,-0.95],[4,-0.9],[10,-0.95]]}
          RectangleSize {type=MovieSceneScriptingFloatChannel, keys [[0,0.26],[4,0.3],[10,0.26]]}
          BottomScale {type=MovieSceneScriptingFloatChannel, const [0.36,1]}
          BottomPartPosition {type=MovieSceneScriptingFloatChannel, const [-0.36,1]}
          OffsetX {type=MovieSceneScriptingFloatChannel, keys [[0,-0.24],[4,-0.15],[10,-0.24]]}
          DropLinesThickness {type=MovieSceneScriptingFloatChannel, const [0.05,1]}
          OpacitySideLines {type=MovieSceneScriptingFloatChannel, keys [[0,1],[2,0.3],[8,0.3],[10,1]]}
          Opacity {type=MovieSceneScriptingFloatChannel, keys [[0,1],[2,0.3],[8,0.3],[10,1]]}
         ]
        }]
      }]
    }
   ]}
 }
]
graphs [
 BindToWeaponEvents {nodes [
   n1 = BindToWeaponEvents!() -> (then, DoBind:bool, EquippedWeapon:Fort Weapon Object Reference) {is=title K2Node_FunctionEntry}
   > Parent: BindToWeaponEvents!(execute, DoBind:bool=@n1, EquippedWeapon:Fort Weapon Object Reference=@n1.EquippedWeapon) -> (then) {is=title K2Node_CallParentFunction}
   > n2 = Cast To FortVehicleModWeaponMortar!(execute, Object=@n1.EquippedWeapon) -> (then, CastFailed, AsFort Vehicle Mod Weapon Mortar:Fort Vehicle Mod Weapon Mortar Object Reference) {is=title K2Node_DynamicCast}
   > n3 = Set WeaponRef(WeaponRef:Fort Vehicle Mod Weapon Mortar Object Reference=@n2) -> (Output_Get:Fort Vehicle Mod Weapon Mortar Object Reference)
   > Bind Event to On Press Trigger(self:Fort Weapon Object Reference=@n3, Delegate=Create Event()) {is=title K2Node_AddDelegate}
   > Bind Event to On Targeting Changed(self:Fort Weapon Object Reference=@n3, Delegate=Create Event()) {is=title K2Node_AddDelegate}
  ]}
 EventGraph {nodes [
   Event Construct()
   > Parent: Construct()
   > PlayAnimation(InAnimation=Get Anim_Intro() -> (Anim_Intro:Widget Animation Object Reference))
   > UpdateCooldownQueryTimer(self:Fort Weapon Reticle Extension Widget Base Object Reference)
   Event Destruct()
   > Parent: Destruct()
   > HandleCooldownTimer!(execute, self:Fort Weapon Reticle Extension Widget Base Object Reference, bActivateTimer:bool=false, TimerInterval:float=0) -> (then)
   n1 = Event OnLocalAmmoChanged!() -> (OutputDelegate, then, LocalCount:int=0, LocalRemaining:int=0) {is=title K2Node_Event}
   > Parent: OnLocalAmmoChanged!(execute, LocalCount:int=@n1.LocalCount, LocalRemaining:int=@n1.LocalRemaining) -> (then) {is=title K2Node_CallParentFunction}
   > UpdateCooldownQueryTimer()
   Event UpdateRemainingCooldown() {is=title K2Node_Event}
   > n2 = GetRemainingCooldownTimer!(execute, self) -> (then, RemainingTime:double, TotalDuration:double)
   > UpdateRemainingRechargeTime!(
    execute
    self:Fort Weapon Reticle Recharge Stack Base Pip Collection Base Object Reference=Get WBP_VehicleMode_Mortar_Collection() -> (WBP_VehicleMode_Mortar_Collection:WBP Vehicle Mode Mortar Collection Object Reference)
    TimeRemaining:float=@n2
    TotalDuration:float=@n2.TotalDuration
    PipPercentMatParamName:Name=Percent
    ShouldSetPipColor:bool=false
    NotReadyColor:Linear Color Structure
    ReadyColor:Linear Color Structure
    PipColorMatParamName:Name=None
   ) -> (then)
   Event UpdateCooldownQueryTimer() {is=title K2Node_Event}
   > n3 = Get!(execute:Exec) -> (then:Exec, else:Exec, WeaponRef:Fort Vehicle Mod Weapon Mortar Object Reference)
   > n4 = Branch(Condition="integer >= integer"(
     A=GetLocalMagazineAmmoCount!(self:Fort Weapon Object Reference=@n3) -> (ReturnValue:int=0)
     B=GetBulletsPerClip!(self:Fort Weapon Object Reference=@n3) -> (ReturnValue:int=0)
    ))
   > HandleCooldownTimer(bActivateTimer=false, TimerInterval=0)
   HandleCooldownTimer(execute=@n4.else, bActivateTimer=true, TimerInterval=0.1)
  ]}
 GetPipCollection {nodes [
   GetPipCollection() {is=title K2Node_FunctionEntry}
   > Return Node!(execute, PipCollection:WBP Reticle Recharge Stack Base Pip Collection Object Reference=Get WBP_VehicleMode_Mortar_Collection())
  ]}
 GetRemainingCooldownTimer {nodes [
   GetRemainingCooldownTimer() {is=title K2Node_FunctionEntry}
   > n1 = Cast To B_VehicleMortarWeapon!(
    execute
    Object=GetEquippedWeapon!(self:Fort Weapon Reticle Extension Widget Base Object Reference) -> (ReturnValue:Fort Weapon Object Reference)
   ) -> (then, CastFailed, AsB Vehicle Mortar Weapon:B Vehicle Mortar Weapon Object Reference) {is=title K2Node_DynamicCast}
   > Return Node!(
    execute
    RemainingTime:double=Get Timer Remaining Time by Handle!(Handle:Timer Handle Structure=Get GrantAmmoTimer!(self:B Vehicle Mortar Weapon Object Reference=@n1) -> (GrantAmmoTimer:Timer Handle Structure)) -> (ReturnValue:float) {is=title K2Node_CallFunction}
    TotalDuration:double=Get RefillAmmoTimer!(self:B Vehicle Mortar Weapon Object Reference=@n1) -> (RefillAmmoTimer:float)
   )
   Return Node(execute=@n1.CastFailed, RemainingTime=0, TotalDuration=0)
  ]}
 SetFireAnimation {nodes [
   SetFireAnimation!() -> (then, Weapon:Fort Weapon Object Reference) {is=title K2Node_FunctionEntry}
   > n1 = Branch(Condition=Get IsTargetting() -> (IsTargetting:bool))
   > PlayAnimationForward(InAnimation=Get Anim_Fire_ADS() -> (Anim_Fire_ADS:Widget Animation Object Reference))
   PlayAnimationForward(execute=@n1.else, InAnimation=Get Anim_Fire() -> (Anim_Fire:Widget Animation Object Reference))
  ]}
 SetTargettingAnimation {nodes [
   n1 = SetTargettingAnimation!() -> (then, bIsTargeting:bool=~) {is=title K2Node_FunctionEntry}
   > n2 = Branch(Condition=@n1)
   n3 = Get Anim_ADS() -> (Anim_ADS:Widget Animation Object Reference)
   > PlayAnimationForward(InAnimation=@n3)
   > Set IsTargetting(IsTargetting:bool=true) -> (Output_Get:bool)
   PlayAnimationReverse(execute=@n2.else, InAnimation=@n3)
   > Set IsTargetting(IsTargetting=false)
  ]}
]
variables [
 IsTargetting {type=boolean}
 WeaponRef {type=FortVehicleModWeaponMortar, default=null}
]
widgets {
 info {
  inheritedWidgetCount=1
  parentClass=/Game/Athena/HUD/reticles/ReticleExtensionWidgets/ChargeStack/WBP_Reticle_RechargeStackBase.WBP_Reticle_RechargeStackBase_C
  rootWidgetClass=&Overlay
  widgetCount=5
 }
 widgets [
  Overlay_0 {widgetClassPath=&Overlay, widget=Overlay_0}
  Reticle {
   widgetClassPath=&Image
   bIsVariable=true
   parent=Overlay_0
   widget=Reticle
   properties {brush {
     bIsDynamicallyLoaded=false
     drawAs=Image
     imageType=FullColor
     mirroring=NoMirror
     resourceName=None
     resourceObject=/VehicleModsUI/Art/Textures/Mortar/MI_UI_VehicleMode_MortarReticle
     tiling=NoTile
     imageSize=<200 200>
     margin=<ltrb 0 0 0 0>
     outlineSettings {
      bUseBrushTransparency=false
      roundingType=HalfHeightRadius
      width=0
      color {colorUseRule=UseColor_Specified, specifiedColor=<rgba 0 0 0 0>}
      cornerRadii=<0 0 0 0>
     }
     tintColor {colorUseRule=UseColor_Specified, specifiedColor=<rgba 1 1 1 1>}
     uVRegion {bIsValid=false, max=<0 0>, min=<0 0>}
    }}
   slot {class=OverlaySlot}
  }
  WBP_VehicleMode_Mortar_Collection {
   widgetClassPath=&/WBP_VehicleMode_Mortar_Collection
   bIsVariable=true
   parent=Overlay_0
   widget=WBP_VehicleMode_Mortar_Collection
   properties {renderTransform {angle=0, scale=<0.4 1>, shear=<0 0>, translation=<20 -56>}}
   slot {class=OverlaySlot, properties {horizontalAlignment=HAlign_Right, verticalAlignment=VAlign_Bottom}}
  }
  WBP_VehicleMods_DisabledWarning {
   widgetClassPath=/VehicleModsUI/Widgets/WBP_VehicleMods_DisabledWarning
   bIsVariable=true
   parent=Overlay_0
   widget=WBP_VehicleMods_DisabledWarning
   slot {class=OverlaySlot, properties {horizontalAlignment=HAlign_Center, verticalAlignment=VAlign_Center}}
  }
  TutorialHighlight {widgetClassPath=&TutorialHighlight}
 ]
}
